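; Stage show actor Blueprint that feeds beat-based light intensity and rotation signals to its static meshes' materials through Custom Primitive Data.
; The construction script writes default signal profiles, scale/offsets and rotation modes to each mesh.
; EventUpdateMaterialCustomPrimitiveData unpacks event settings into those parameters, writing rotation only when UseRotationSignal is set.
; ApplyGroupSignalOffset staggers beat offsets by the actor's group index using all-at-once, linear or inside-out patterns.
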(event EventUpdateMaterialCustomPrimitiveData
  (params (EventSettings :type Stage%20Show%20Event%20Settings%20Structure%20%28by%20ref%29))
  (bind (_riseduration _riseeasing _fallduration _falleasing _beatmultiplier _beatoffset _peakvalue _minimumvalue) (Utilities|Struct|BreakStageShowSignalData (.RotationSignal (Utilities|Struct|BreakStageShowEventSettings EventSettings))))
  (bind (_lightintensitysignal _rotationsignal _rotationsettings _grouppatternsettings) (Utilities|Struct|BreakStageShowEventSettings EventSettings))
  (bind (_riseduration_2 _riseeasing_2 _fallduration_2 _falleasing_2 _beatmultiplier_2 _beatoffset_2 _peakvalue_2 _minimumvalue_2) (Utilities|Struct|BreakStageShowSignalData _lightintensitysignal))
  (bind (_pattern_index _playback_mode) (CallFunction|GetRotationSettingsMaterialFloatParameters _rotationsettings))
  (bind _staticmeshes (Variables|Default|GetStaticMeshes))
  (if (Utilities|Array|IsNotEmpty _staticmeshes)
    (for _array_element _staticmeshes
      (Utilities|IsValid _array_element
        (:Is%20Valid
          ; Intensity Parameters
          (Rendering|Material|SetVectorParameterforCustomPrimitiveData _array_element (Variables|MaterialParametersNamesConstants|GetIntensityProfileParam) (Math|Vector4|MakeVector4 _riseduration_2 _riseeasing_2 _fallduration_2 _falleasing_2))
          (Rendering|Material|SetVectorParameterforCustomPrimitiveData _array_element (Variables|MaterialParametersNamesConstants|GetIntensityScaleOffsetParam) (Math|Vector4|MakeVector4 _beatmultiplier_2 (Default|ApplyGroupSignalOffset _grouppatternsettings _lightintensitysignal) _peakvalue_2 _minimumvalue_2))
          (if (Variables|Default|GetUseRotationSignal)
            ; Rotation Parameters
            (Rendering|Material|SetVectorParameterforCustomPrimitiveData _array_element (Variables|MaterialParametersNamesConstants|GetRotationProfileParam) (Math|Vector4|MakeVector4 _riseduration _riseeasing _fallduration _falleasing))
            (Rendering|Material|SetVectorParameterforCustomPrimitiveData _array_element (Variables|MaterialParametersNamesConstants|GetRotationScaleOffsetParam) (Math|Vector4|MakeVector4 _beatmultiplier (Default|ApplyGroupSignalOffset _grouppatternsettings _rotationsignal) _peakvalue _minimumvalue))
            (Rendering|Material|SetScalarParameterforCustomPrimitiveData _array_element (Variables|MaterialParametersNamesConstants|GetRotationPatternParam) _pattern_index)
            (Rendering|Material|SetScalarParameterforCustomPrimitiveData _array_element (Variables|MaterialParametersNamesConstants|GetRotationPlaybackParam) _playback_mode)))))))

(fn ApplyGroupSignalOffset
  (params (Group%20Pattern%20Data :type Stage%20Show%20Group%20Pattern%20Data%20Structure%20%28by%20ref%29) (Signal%20Data :type Stage%20Show%20Signal%20Data%20Structure%20%28by%20ref%29))
  (bind (_offsettype _duration) (Utilities|Struct|BreakStageShowGroupPatternData Group%20Pattern%20Data))
  (bind (_riseduration _riseeasing _fallduration _falleasing _beatmultiplier _beatoffset) (Utilities|Struct|BreakStageShowSignalData Signal%20Data))
  (switch Utilities|FlowControl|Switch|SwitchonEStageGroupOffset _offsettype
    (:AllAtOnce
      (return _beatoffset))
    (:Linear
      ; Linear
      (return (+ (* (Math|Float|SafeDivide (* (/ 1.0 (Math|Conversions|ToFloat(Integer) (Variables|Default|GetGroupSize))) (Variables|Default|GetActorGroupIndex)) _beatmultiplier) _duration) _beatoffset)))
    (:InsideOut
      ; Inside Out
      (return (+ _beatoffset (* _duration (Math|Float|SafeDivide (* (Math|Float|Floor (Math|Float|Absolute(Float) (- (/ (Math|Conversions|ToFloat(Integer) (- (Variables|Default|GetGroupSize) 1)) 2.0) (Variables|Default|GetActorGroupIndex)))) (/ 1.0 (Math|Conversions|ToFloat(Integer) (/ (Variables|Default|GetGroupSize) 2)))) _beatmultiplier)))))))

(fn GetRotationSettingsMaterialFloatParameters
  (params (StageShowRotationSettings :type Stage%20Show%20Rotation%20Settings%20Structure%20%28by%20ref%29))
  (bind (_pattern _animationplayback) (Utilities|Struct|BreakStageShowRotationSettings StageShowRotationSettings))
  (return (Utilities|Select :Index _pattern :None 0.0 :Bounce 1.0 :Circle 2.0) (Utilities|Select :Index _animationplayback :PingPong 1.0 :Linear 0.0)))

(fn InitializeActor)

(fn ConstructionScript
  (bind _returnvalue (Math|Vector4|MakeVector4 4.0 1.0 2.0 1.0))
  (bind _stageshowrotationsettings (Utilities|Struct|MakeStageShowRotationSettings "None"))
  (bind (_pattern_index _playback_mode) (CallFunction|GetRotationSettingsMaterialFloatParameters _stageshowrotationsettings))
  (bind _staticmeshes (Variables|Default|GetStaticMeshes))
  (if (Utilities|Array|IsNotEmpty _staticmeshes)
    (for _array_element _staticmeshes
      (Utilities|IsValid _array_element
        (:Is%20Valid
          (Rendering|Material|SetVectorParameterforDefaultCustomPrimitiveData _array_element (Variables|MaterialParametersNamesConstants|GetIntensityScaleOffsetParam) (Math|Vector4|MakeVector4 1.0 0.0 1.0 1.0))
          (Rendering|Material|SetVectorParameterforDefaultCustomPrimitiveData _array_element (Variables|MaterialParametersNamesConstants|GetRotationScaleOffsetParam) (Math|Vector4|MakeVector4 1.0))
          (Rendering|Material|SetVectorParameterforDefaultCustomPrimitiveData _array_element (Variables|MaterialParametersNamesConstants|GetIntensityProfileParam) _returnvalue)
          (Rendering|Material|SetVectorParameterforDefaultCustomPrimitiveData _array_element (Variables|MaterialParametersNamesConstants|GetRotationProfileParam) _returnvalue)
          (Rendering|Material|SetScalarParameterforDefaultCustomPrimitiveData _array_element (Variables|MaterialParametersNamesConstants|GetRotationPatternParam) _pattern_index)
          (Rendering|Material|SetScalarParameterforDefaultCustomPrimitiveData _array_element (Variables|MaterialParametersNamesConstants|GetRotationPlaybackParam) _playback_mode))))))
