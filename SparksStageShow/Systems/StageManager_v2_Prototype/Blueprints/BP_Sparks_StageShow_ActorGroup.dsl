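; BP_Sparks_StageShow_ActorGroup is a prototype Stage Manager v2 actor that manages a list of stage show actors as one group.
; On BeginPlay it gives each actor its index and the group size, then initializes it.
; PlayEvent pushes the supplied event settings into each valid actor's material custom primitive data.
; UpdateActorsCustomPrimitiveData does the same with the group's stored EventSettings.
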(event EventBeginPlay
  (Default|InitializeActors))

(event Collision|EventActorBeginOverlap
  (params (OtherActor :type Actor%20Object%20Reference)))

(event EventTick
  (params (DeltaSeconds :type Float%20%28single-precision%29)))

(fn InitializeActors
  (bind _stageshowactors (Variables|Default|GetStageShowActors))
  (for (_array_element _array_index) _stageshowactors
    (Variables|/SparksStageShow/Systems/StageManager_v2_Prototype/Blueprints/BP_Sparks_StageShow_Actor.BP_Sparks_StageShow_Actor_C|SetActorGroupIndex _array_index _array_element)
    (Variables|/SparksStageShow/Systems/StageManager_v2_Prototype/Blueprints/BP_Sparks_StageShow_Actor.BP_Sparks_StageShow_Actor_C|SetGroupSize (Utilities|Array|Length _stageshowactors) _array_element)
    (Default|InitializeActor _array_element)))

(fn PlayEvent
  (params (Stage%20Show%20Event%20Settings :type Stage%20Show%20Event%20Settings%20Structure))
  (for _array_element (Variables|Default|GetStageShowActors)
    (Utilities|IsValid _array_element
      (:Is%20Valid
        (Class|StageShowActorBase|UpdateMaterialCustomPrimitiveData _array_element Stage%20Show%20Event%20Settings)))))

(fn UpdateActorsCustomPrimitiveData
  (for _array_element (Variables|Default|GetStageShowActors)
    (Utilities|IsValid _array_element
      (:Is%20Valid
        (Class|StageShowActorBase|UpdateMaterialCustomPrimitiveData _array_element (Variables|Default|GetEventSettings))))))

(fn ConstructionScript)
