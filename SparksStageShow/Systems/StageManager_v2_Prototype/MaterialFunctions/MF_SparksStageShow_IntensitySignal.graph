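# Material function that outputs the light intensity signal for the Sparks Stage Show.
# It passes the shared MeasureLifetime collection parameter and a Beats Per Measure input, defaulting to four, into MF_SparksStageShow_SignalProcessor and returns its Signal.
# Two vector parameters read from Custom Primitive Data, allowing per-primitive control.
# Their channels are named for beat timing and range, and for rise/fall time and easing.
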MaterialFunction /SparksStageShow/Systems/StageManager_v2_Prototype/MaterialFunctions/MF_SparksStageShow_IntensitySignal
graph {
 properties {description=Outputs the Stage Show light intensity signal., userExposedCaption=Stage Show Intensity Signal}
 nodes [
  n1 = FunctionInput() {bUsePreviewValueAsDefault=true, inputName=Beats Per Measure, inputType=FunctionInput_Scalar, previewValue=<4 0 0 1>}
  n2 = VectorParameter() {
   bUseCustomPrimitiveData=true
   parameterName=Intensity Signal Scale and Offset
   channelNames {a=Min Value, b=Peak Value, g=Beat Offset, r=Beat Multiplier}
  }
  n3 = VectorParameter() {
   bUseCustomPrimitiveData=true
   parameterName=Intensity Signal Profile
   primitiveDataIndex=4
   channelNames {a=Fall Easing, b=Fall Time, g=Rise Easing, r=Rise Time}
  }
  FunctionOutput(None=MaterialFunctionCall!(
    Measure Signal=CollectionParameter() {collection=&/MPC_SparksStageShow, parameterName=MeasureLifetime}
    Beats Per Measure=@n1
    Beat Multiplier=@n2.Beat Multiplier
    BeatOffset=@n2.Beat Multiplier
    Peak Value=@n2.Beat Multiplier
    Min Value=@n2.Beat Multiplier
    Rise Time=@n3.Rise Time
    Rise Easing=@n3.Rise Time
    Fall Time=@n3.Rise Time
    Fall Easing=@n3.Rise Time
    AnimationMode
   ) {materialFunction=&/MF_SparksStageShow_SignalProcessor}.Signal) {outputName=Signal}
 ]
}
