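# Material function for the prototype Stage Show system that outputs an RGB rotation signal from a curve atlas of rotation patterns.
# MeasureLifetime from the Stage Show parameter collection, a Beats Per Measure input and custom primitive data parameters feed the shared SignalProcessor function.
# The processed signal becomes the atlas U coordinate.
# A per-primitive curve index picks the V row at the center of one of eight slots.
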MaterialFunction /SparksStageShow/Systems/StageManager_v2_Prototype/MaterialFunctions/MF_SparksStageShow_RotationSignal
graph {
 properties {description=Outputs the Stage Show rotation signal., userExposedCaption=Stage Show Rotation Signal}
 nodes [
  n1 = CollectionParameter() {collection=&/MPC_SparksStageShow, parameterName=MeasureLifetime}
  n2 = FunctionInput() {bUsePreviewValueAsDefault=true, inputName=Beats Per Measure, inputType=FunctionInput_Scalar, previewValue=<4 0 0 1>}
  n3 = VectorParameter() {
   bUseCustomPrimitiveData=true
   parameterName=Rotation Signal Scale and Offset
   primitiveDataIndex=8
   channelNames {a=Min Value, b=Peak Value, g=Beat Offset, r=Beat Multiplier}
  }
  n4 = VectorParameter() {
   bUseCustomPrimitiveData=true
   parameterName=Rotation Signal Profile
   primitiveDataIndex=12
   channelNames {a=Fall Easing, b=Fall Time, g=Rise Easing, r=Rise Time}
   defaultValue=<rgba 4 1 2 1>
  }
  n5 = ScalarParameter() {bUseCustomPrimitiveData=true, defaultValue=1, parameterName=RotationAnimationMode, primitiveDataIndex=16}
  n6 = Divide() {constB=8}
  n7 = Multiply(A=ScalarParameter() {bUseCustomPrimitiveData=true, defaultValue=1, parameterName=RotationCurveIndex, primitiveDataIndex=17}, B=@n6)
  n8 = Multiply(A=@n6) {constB=0.5}
  n9 = TextureObject() {
   samplerType=SAMPLERTYPE_LinearColor
   texture=/SparksStageShow/Systems/StageManager_v2_Prototype/CurveAtlas/Rotation/CA_StageShow_RotationPatterns
  }
  n10 = TextureSample(
   UVs=MaterialFunctionCall(
    X=MaterialFunctionCall!(
     Measure Signal=@n1
     Beats Per Measure=@n2
     Beat Multiplier=@n3.Beat Multiplier
     BeatOffset=@n3.Beat Multiplier
     Peak Value=@n3.Beat Multiplier
     Min Value=@n3.Beat Multiplier
     Rise Time=@n4.Rise Time
     Rise Easing=@n4.Rise Time
     Fall Time=@n4.Rise Time
     Fall Easing=@n4.Rise Time
     AnimationMode=@n5
    ) {materialFunction=&/MF_SparksStageShow_SignalProcessor}.Signal
    Y=Add(A=@n7, B=@n8)
   ) {materialFunction=&MakeFloat2}.Result
   Tex=@n9
  ) {
   samplerType=SAMPLERTYPE_LinearColor
   texture=&CA_StageShow_RotationPatterns
  }
  FunctionOutput(None=@n10.RGB)
  TextureProperty(Tex=@n9)
 ]
}
