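# Material function that rotates a normalized rotation pattern vector to follow the direction of stage show actors that receive a rotation signal.
# It builds a basis from the direction input using cross products with world up, with a SafeNormalize fallback for degenerate cases, and transforms the pattern through it.
# The useStageRotationSignal static switch gates the transformed pattern and three basis outputs, which become world axes when off.
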MaterialFunction /SparksStageShow/Systems/StageManager_v2_Prototype/MaterialFunctions/MF_SparksStageShow_TransformRotPattern
graph {
 properties {
  description=Used to transform rotation vector based on stage show actor direction. To be used on stage show actors that can receive rotation signal
  userExposedCaption=StageShow Transform Rotation Pattern
 }
 nodes [
  n1 = Time()
  n2 = Sine(None=@n1)
  n3 = Cosine(None=@n1)
  n4 = Constant() {r=0.5}
  n5 = NamedRerouteDeclaration(None=Normalize(VectorInput=FunctionInput() {bUsePreviewValueAsDefault=true, inputName=Normalized Direction, previewValue=<1 0 0 1>})) {name=NormalizedDirection, nodeColor=<rgba 1 0.3984 0 1>}
  n6 = Constant3Vector() {constant=<rgba 1 0 0 0>}
  n7 = MaterialFunctionCall(Vector=CrossProduct(A=Constant3Vector() {constant=<rgba 0 0 1 0>}, B=@n5), Default=Normalize(VectorInput=CrossProduct(A=@n5, B=@n6))) {materialFunction=&SafeNormalize}
  n8 = CrossProduct(A=@n5, B=@n7.Result)
  n9 = MaterialFunctionCall(
   VectorToTransform=FunctionInput(Preview=MaterialFunctionCall(X=@n2, Y=@n3, Z=@n4) {materialFunction=&MakeFloat3}.Result) {bUsePreviewValueAsDefault=true, inputName=Normalized Rotation Pattern, previewValue=<1 0 0 1>}
   BasisX=@n5
   BasisY=@n7.Result
   BasisZ=@n8
  ) {materialFunction=&Transform3x3Matrix}
  FunctionOutput(None=StaticSwitchParameter(True=Normalize(VectorInput=@n9.Transformed Vector), False=NamedRerouteUsage()) {defaultValue=true, group=Stage Show Signal, parameterName=useStageRotationSignal}) {outputName=Transformed Pattern}
  FunctionOutput(None=StaticSwitchParameter(True=@n5, False=Constant3Vector() {constant=<rgba 1 0 0 0>}) {defaultValue=true, group=Stage Show Signal, parameterName=useStageRotationSignal}) {outputName=BasisX, sortPriority=1}
  FunctionOutput(None=StaticSwitchParameter(True=@n7.Result, False=Constant3Vector() {constant=<rgba 0 1 0 0>}) {defaultValue=true, group=Stage Show Signal, parameterName=useStageRotationSignal}) {outputName=BasisY, sortPriority=2}
  FunctionOutput(None=StaticSwitchParameter(True=@n8, False=Constant3Vector() {constant=<rgba 0 0 1 0>}) {defaultValue=true, group=Stage Show Signal, parameterName=useStageRotationSignal}) {outputName=BasisZ, sortPriority=3}
 ]
}
