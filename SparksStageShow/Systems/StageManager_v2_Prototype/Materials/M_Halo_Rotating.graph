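# Unlit additive halo material for Sparks stage show lights, combining a spiky flare and a corona tinted by the show's primary color.
# A rotation signal turns a per-primitive resting direction into the beam direction that offsets the camera-facing sprite.
# View alignment with that direction drives brightness and shrinks hidden flares, alongside intensity-signal and proximity masks.
# World position offset builds the sprite and nudges it cameraward against clipping, while screen position masks the corona.
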Material /SparksStageShow/Systems/StageManager_v2_Prototype/Materials/M_Halo_Rotating
graph {
 properties {
  bAutomaticallySetUsageInEditor=false
  blendMode=BLEND_Additive
  maxWorldPositionOffsetDisplacement=200
  shadingModel=MSM_Unlit
  usedShadingModels=Unlit
 }
 nodes [
  n1 = ScalarParameter() {defaultValue=200, desc=Distance where flare will have zero opacity, group=Flare, parameterName=FadeOffset, sliderMax=500, sortPriority=10}
  n2 = Subtract(A=Length(None=ObjectPositionWS() {originType=CameraRelative}) {is=class MaterialExpressionLength}, B=@n1)
  n3 = ScalarParameter() {
   defaultValue=50
   desc=Size of the transition zone where the flare fades with proximity.
   group=Flare
   parameterName=FadeDistance
   sliderMax=500
   sortPriority=9
  }
  NamedRerouteDeclaration(None=Saturate(None=Divide(A=@n2, B=@n3) {constB=400})) {name=ProximityMask, nodeColor=<rgba 1 0 0.2344 1>}
  n4 = ScalarParameter() {
   defaultValue=0.2
   desc=Brightness multiplier when not facing the flare
   group=Flare
   parameterName=ViewDirBoostMin
   sliderMax=5
   sortPriority=7
  }
  n5 = ScalarParameter() {
   defaultValue=1
   desc=Brightness multiplier when directly facing the flare
   group=Flare
   parameterName=ViewDirBoostMax
   sliderMax=5
   sortPriority=8
  }
  n6 = Saturate(None=NamedRerouteUsage())
  n7 = Power(Base=@n6) {constExponent=4}
  NamedRerouteDeclaration(None=Multiply(A=LinearInterpolate(A=@n4, B=@n5, Alpha=@n7), B=SmoothStep(Value=@n6) {constMax=0.2})) {name=ViewDirMask, nodeColor=<rgba 0.4297 1 0 1>}
  NamedRerouteDeclaration(None=CollectionParameter() {collection=/SparksStageShow/Systems/StageManager_v2_Prototype/MaterialFunctions/MPC_SparksStageShow, parameterName=PrimaryColor}) {name=Stage Show Color, nodeColor=<rgba 0 0.5781 1 1>}
  n8 = ScalarParameter() {defaultValue=3, desc=Base flare brightness, group=Flare, parameterName=FlareIntensity, sliderMax=300, sortPriority=3}
  n9 = TextureSampleParameter2D(UVs=MaterialFunctionCall(
    UVs=TextureCoordinate()
    Texture Scale=ScalarParameter() {defaultValue=1, desc=Center flare scale, group=Flare, parameterName=FlareScale, sliderMax=1, sortPriority=1}
   ) {materialFunction=&ScaleUVsByCenter}.UVs) {group=Flare Settings, parameterName=FlareMask, sortPriority=-2, texture=/Game/Effects/Fort_Effects/Textures/Flares/T_Flare_Spikeball}
  n10 = NamedRerouteUsage()
  n11 = ScalarParameter() {defaultValue=0.3, group=Flare, parameterName=FlareSaturation, sliderMax=1, sortPriority=3}
  n12 = ObjectPositionWS() {originType=CameraRelative}
  n13 = MaterialFunctionCall(Float3=Transform(None=Normalize(VectorInput=@n12)) {transformSourceType=TRANSFORMSOURCE_World, transformType=TRANSFORM_View}) {materialFunction=&BreakOutFloat3Components}
  n14 = MaterialFunctionCall(X=@n13.R, Y=Multiply(A=@n13.G) {constB=-1}) {materialFunction=&MakeFloat2}
  n15 = ScalarParameter() {
   defaultValue=0.8
   desc="Radius of the view position mask.\r\nSmaller values result in smaller fractions of corona being visible."
   group=Corona View Position Mask
   parameterName=CoronaViewAng_MaskRadius
   sliderMax=1
   sliderMin=0.1
   sortPriority=0
  }
  n16 = SphereMask(A=Add(A=VertexInterpolator(VS=@n14.Result).PS, B=TextureCoordinate()), B=Constant() {r=0.5}, Radius=@n15) {attenuationRadius=0.6, hardnessPercent=0}
  n17 = ScalarParameter() {
   defaultValue=1
   desc=Boost corona brightness based on view space pos
   group=Corona View Position Mask
   parameterName=CoronaViewAng_Boost
   sliderMax=3
   sortPriority=2
  }
  n18 = ScalarParameter() {
   defaultValue=0.1
   desc=Minimum brightness where corona falls outside view pos mask
   group=Corona View Position Mask
   parameterName=CoronaViewAngleMin
   sliderMax=1
   sortPriority=2
  }
  n19 = TextureCoordinate()
  n20 = Constant() {r=0.5}
  n21 = LinearInterpolate(Alpha=ScalarParameter() {defaultValue=0.8, desc=Corona scale, group=Flare, parameterName=CoronaScale, sliderMax=1, sortPriority=4}) {constA=-0.3, constB=0.7}
  n22 = Multiply(
   A=Max(A=Multiply(A=@n16, B=@n17) {constB=1.5}, B=@n18) {constB=0.1}
   B=TextureSample(UVs=SphereMask(A=@n19, B=@n20, Radius=@n20, Hardness=@n21)) {texture=/Game/Effects/Fort_Effects/Textures/Grad/T_Halo_Corona_Gradient}.RGB
  )
  n23 = ScalarParameter() {defaultValue=0.3, desc=Corona base brightness, group=Flare, parameterName=CoronaIntensity, sliderMax=10, sortPriority=5}
  n24 = NamedRerouteUsage()
  n25 = ScalarParameter() {defaultValue=0.3, desc=How much color tinting will be applied to corona, group=Flare, parameterName=CoronaSaturation, sortPriority=6}
  n26 = Add(
   A=Multiply(A=Multiply(A=@n8, B=@n9.RGB) {constA=4}, B=Desaturation(None=@n10, Fraction=@n11) {luminanceFactors=<rgba 0.2126 0.7152 0.0722 1>})
   B=Multiply(A=Multiply(A=@n22, B=@n23), B=Desaturation(None=@n24, Fraction=@n25) {luminanceFactors=<rgba 0.2126 0.7152 0.0722 1>})
  )
  n27 = NamedRerouteUsage()
  n28 = NamedRerouteUsage()
  n29 = VertexInterpolator(VS=Multiply(A=Multiply(A=@n27, B=@n28), B=NamedRerouteUsage()))
  NamedRerouteDeclaration(None=MaterialFunctionCall(Emissive=Multiply(A=@n26, B=@n29.PS)) {materialFunction=&MF_FX_EmissiveAdjustments}.Result) {name=Emissive, nodeColor=<rgba 0 0.8359 1 1>}
  n30 = Transform(None=NamedRerouteUsage()) {transformSourceType=TRANSFORMSOURCE_Local}
  n31 = Add(A=ObjectPositionWS(), B=Multiply(A=@n30, B=NamedRerouteUsage()))
  n32 = Normalize(VectorInput=ObjectPositionWS() {originType=CameraRelative})
  n33 = NamedRerouteDeclaration(None=DotProduct(A=@n30, B=Multiply(A=@n32) {constB=-1})) {name=ViewAlignement, nodeColor=<rgba 0 1 0.8359 1>}
  n34 = Multiply(A=Multiply(A=NamedRerouteUsage(), B=Ceil(None=@n33)), B=NamedRerouteUsage())
  n35 = Multiply(
   A=ScalarParameter() {defaultValue=150, desc=General flare size, group=Flare, parameterName=Size, sortPriority=0}
   B=LinearInterpolate(Alpha=Ceil(None=@n34)) {constA=0.01}
  )
  n36 = ScalarParameter() {
   defaultValue=30
   desc="Offsets flare to camera direction. \r\nUse to reduce clipping with near objects."
   group=Flare
   parameterName=CameraOffset
  }
  n37 = ScalarParameter() {defaultValue=200, desc="Minimum distance from camera. \r\nUse this to reduce camera clipping.", group=Flare, parameterName=CameraPadding}
  n38 = ObjectPositionWS()
  NamedRerouteDeclaration(None=Add(
    A=MaterialFunctionCall!(Center Location=@n31, 0-1 UVs, XY scale=@n35, OverridePivot, "Normals (Optional)") {materialFunction=&Sprite}.World position offset
    B=MaterialFunctionCall(Offset Amount=@n36, Clamp Padding=@n37, World Position=@n38) {materialFunction=/Engine/Functions/Engine_MaterialFunctions02/WorldPositionOffset/CameraOffset}.Clamped Camera Offset
   )) {name=WPO, nodeColor=<rgba 1 0 0.5156 1>}
  NamedRerouteDeclaration(None=MaterialFunctionCall!(Beats Per Measure) {materialFunction=/SparksStageShow/Systems/StageManager_v2_Prototype/MaterialFunctions/MF_SparksStageShow_IntensitySignal}.Signal) {name=Intensity Signal, nodeColor=<rgba 0.9219 1 0 1>}
  NamedRerouteDeclaration(None=ScalarParameter() {parameterName=PosOffset}) {name=PosOffset, nodeColor=<rgba 0 1 0.7891 1>}
  n39 = MaterialFunctionCall!(Beats Per Measure) {materialFunction=/SparksStageShow/Systems/StageManager_v2_Prototype/MaterialFunctions/MF_SparksStageShow_RotationSignal}
  n40 = VectorParameter() {
   bUseCustomPrimitiveData=true
   group=Flare Settings
   parameterName=LocalRestingDirection
   primitiveDataIndex=18
   sortPriority=-1
   channelNames {a=-, b=Z, g=Y, r=X}
   defaultValue=<rgba 1 0 0 0>
  }
  NamedRerouteDeclaration(None=MaterialFunctionCall!(Normalized Rotation Pattern=Normalize(VectorInput=@n39.Result), Normalized Direction=Normalize(VectorInput=@n40)) {materialFunction=/SparksStageShow/Systems/StageManager_v2_Prototype/MaterialFunctions/MF_SparksStageShow_TransformRotPattern}.Transformed Pattern) {name=TargetDirection, nodeColor=<rgba 0 0.8125 1 1>}
  NamedRerouteUsage()
  NamedRerouteUsage()
  n41 = PixelDepth()
  OneMinus(None=SphereMask(A=Subtract(A=@n41.R) {constB=300}, B=Constant(), Radius=Constant() {r=256}) {hardnessPercent=10})
 ]
}
