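# Additive unlit spotlight beam material for the SparksStageShow StageManager v2 prototype.
# World position offset billboards and shapes the beam by length, width and angle, optionally rotating it via a stage show rotation signal and custom-primitive-data resting direction.
# Emission combines UV falloff, attenuation, the show's intensity signal and collection PrimaryColor.
# Near-camera and view-angle fades soften it; mobile skips depth fade and uses a brightness multiplier.
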Material /SparksStageShow/Systems/StageManager_v2_Prototype/Materials/M_Spotlight_Beam_LookAt
graph {
 properties {
  bAutomaticallySetUsageInEditor=false
  bCastRayTracedShadows=false
  bSubstrateRoughnessTracking=false
  bUseLightmapDirectionality=false
  blendMode=BLEND_Additive
  maxWorldPositionOffsetDisplacement=8000
  shadingModel=MSM_Unlit
  usedShadingModels=Unlit
 }
 nodes [
  n1 = ScalarParameter() {
   defaultValue=0.07
   desc=Set the beam end tip roundness
   group=Light Beam Settings
   parameterName=Tip Roundness
   sliderMax=0.25
   sortPriority=5
  }
  NamedRerouteDeclaration(None=@n1) {name=Roundness, nodeColor=<rgba 0 1 0.0648 1>}
  n2 = TextureCoordinate()
  n3 = Multiply(A=Distance(A=ComponentMask(None=@n2) {r=true}, B=Constant() {r=0.5})) {constB=2}
  n4 = Multiply(A=OneMinus(None=Power(Base=@n3)), B=OneMinus(None=ComponentMask(None=@n2) {g=true}))
  NamedRerouteDeclaration(None=LinearInterpolate(B=Saturate(None=@n4), Alpha=NamedRerouteUsage()) {constA=1}) {name=RoundedOffset, nodeColor=<rgba 1 0.8587 0.0832 1>}
  n5 = ScalarParameter() {defaultValue=500, desc="Near fade, use to adjust camera clipping", group=Light Beam Camera Fade, parameterName=NearFade, sortPriority=31}
  NamedRerouteDeclaration(None=@n5) {name=NearFade, nodeColor=<rgba 1 0 0.0703 1>}
  n6 = ScalarParameter() {
   defaultValue=0.75
   desc=The beam gradually fades out when front/back faced directly. Use higher values for a harsher gradient
   group=Light Beam Camera Fade
   parameterName=ViewAngleFade
   sortPriority=31
  }
  NamedRerouteDeclaration(None=@n6) {name=ViewAngleFade, nodeColor=<rgba 0 0.3906 1 1>}
  n7 = ScalarParameter() {
   defaultValue=1
   desc=Default HDRMult is too bright for mobile. This allows to reduce for mobile when HDR is high.
   group=Light Beam Settings
   parameterName=MobileBrightnessMultiplier
  }
  NamedRerouteDeclaration(None=@n7) {name=MobileBrightnessMultiplier, nodeColor=<rgba 0.7109 1 0 1>}
  n8 = VectorParameter() {
   bUseCustomPrimitiveData=true
   desc=Base orientation of the stage show actor
   group=Light Beam Pos/Rot Offsets
   parameterName=LocalRestingDirection
   primitiveDataIndex=18
   channelNames {a=-, b=Z, g=Y, r=X}
   defaultValue=<rgba 1 0 0 0>
  }
  NamedRerouteDeclaration(None=@n8) {name=LocalRestingDirection, nodeColor=<rgba 0.3125 0 1 1>}
  n9 = VectorParameter() {
   desc="Custom pivot offset that's used for stage show rotation.\r\nUse to align the rotations with other elements of a multi part object"
   group=Light Beam Pos/Rot Offsets
   parameterName=PivotOffset
   channelNames {a=-, b=Z, g=Y, r=X}
  }
  NamedRerouteDeclaration(None=@n9) {name=PivotOffset, nodeColor=<rgba 0.0078 1 0 1>}
  n10 = VectorParameter() {
   desc="Custom position offset that's used for stage show rotation.\r\nUse to align the position with other elements of a multi part object"
   group=Light Beam Pos/Rot Offsets
   parameterName=PositionOffset
   channelNames {a=-, b=Z, g=Y, r=X}
  }
  NamedRerouteDeclaration(None=@n10) {name=PositionOffset, nodeColor=<rgba 1 0 0.3281 1>}
  n11 = LocalPosition() {includedOffsets=ExcludeOffsets}
  n12 = NamedRerouteUsage()
  NamedRerouteDeclaration(None=TransformPosition(None=Add(A=@n11.XYZ, B=@n12)) {transformType=TRANSFORMPOSSOURCE_TranslatedWorld}) {name=WorldOffsettedPos, nodeColor=<rgba 0.2422 0 1 1>}
  n13 = Constant3Vector() {constant=<rgba 1 0 0 0>}
  n14 = MaterialFunctionCall!(Beats Per Measure) {materialFunction=/SparksStageShow/Systems/StageManager_v2_Prototype/MaterialFunctions/MF_SparksStageShow_RotationSignal}
  n15 = NamedRerouteDeclaration(None=MaterialFunctionCall!(Normalized Rotation Pattern=Normalize(VectorInput=@n14.Result), Normalized Direction=NamedRerouteUsage()) {materialFunction=/SparksStageShow/Systems/StageManager_v2_Prototype/MaterialFunctions/MF_SparksStageShow_TransformRotPattern}.Transformed Pattern) {name=TransformedDirection, nodeColor=<rgba 0 1 0.6953 1>}
  n16 = ArccosineFast!(None=DotProduct(A=@n13, B=@n15))
  n17 = CrossProduct(A=@n13, B=@n15)
  n18 = MaterialFunctionCall!(
   RotationAngle=Multiply(A=Multiply(A=@n16) {constB=57.29, desc=Radians to Degrees}) {constB=0.0028, desc=Convert 360 range to 0-1}
   NormalizedRotationAxis=Transform(None=MaterialFunctionCall(Vector=@n17) {materialFunction=&SafeNormalize}.Result) {transformSourceType=TRANSFORMSOURCE_Local}
   PivotPoint=TransformPosition(None=NamedRerouteUsage()) {transformType=TRANSFORMPOSSOURCE_TranslatedWorld}
   Position=NamedRerouteUsage()
   Normal
   TangentU
   TangentV
  ) {materialFunction=/Game/Characters/Player/Male/Medium/Bodies/M_MED_Inferno/FX/Mat/RotateVertexAboutAxis}
  n19 = NamedRerouteUsage()
  n20 = MaterialFunctionCall() {materialFunction=&TangentBasis}
  n21 = Multiply(A=SmoothStep(Value=NamedRerouteUsage()) {constMax=0.1, constValue=1}, B=NamedRerouteUsage()) {constA=1}
  n22 = Multiply(A=Tangent!(None=NamedRerouteUsage())) {constB=2}
  n23 = NamedRerouteUsage()
  n24 = MaterialFunctionCall!(
   Camera Relative World Position=StaticSwitchParameter(True=Add(A=@n18.WPO, B=@n19), False=@n19) {defaultValue=true, parameterName=isRotating}
   TangentU_WS=StaticSwitchParameter(True=@n18.TangentU, False=@n20.X U) {defaultValue=true, parameterName=isRotating}
   TangentV_WS=StaticSwitchParameter(True=@n18.TangentV, False=@n20.Y V) {defaultValue=true, parameterName=isRotating}
   UVs for Thickness
   UVs for Projection
   Input Expand U or V UV Channel
   Lenght=Multiply(A=NamedRerouteUsage(), B=@n21)
   WidthBase=@n23
   WidthTip=Add(A=@n23, B=Multiply(A=@n21, B=@n22))
   UVs For NormalShading
  ) {materialFunction=/SparksStageShow/Systems/StageManager_v2_Prototype/MaterialFunctions/MF_MojoLineBillboard}
  n25 = Add(A=@n18.WPO, B=@n24.WPO)
  n26 = NamedRerouteUsage()
  n27 = WorldPosition() {worldPositionShaderOffset=WPT_CameraRelativeNoOffsets}
  NamedRerouteDeclaration(None=Add(A=StaticSwitchParameter(True=@n25, False=@n24.WPO) {defaultValue=true, parameterName=isRotating}, B=Subtract(A=@n26, B=@n27.XYZ))) {name=Final WPO, nodeColor=<rgba 1 0.1172 0 1>}
  n28 = TextureCoordinate()
  n29 = Distance(A=ComponentMask(None=@n28) {r=true}, B=Constant() {r=0.5})
  n30 = SmoothStep(Max=NamedRerouteUsage(), Value=OneMinus(None=Multiply(A=@n29) {constB=2})) {constMax=1.5}
  n31 = ComponentMask(None=@n28) {g=true}
  n32 = Saturate(None=SmoothStep(Max=NamedRerouteUsage(), Value=@n31))
  n33 = Power(Base=Multiply(A=OneMinus(None=@n31)) {constB=10})
  n34 = NamedRerouteUsage()
  n35 = Divide(B=Add(A=Multiply(A=@n33, B=@n34)) {constA=1})
  n36 = NamedRerouteUsage()
  n37 = Multiply(A=Multiply(A=Saturate(None=@n30), B=Multiply(A=@n32, B=@n35)), B=Multiply(A=NamedRerouteUsage(), B=VertexInterpolator(VS=@n36).PS)) {constB=10}
  n38 = Multiply(A=@n37, B=DepthFade())
  n39 = QualitySwitch(Default=@n38, Low=@n37, High=@n38, Medium=@n38)
  n40 = Multiply(A=NamedRerouteUsage(), B=@n37)
  n41 = NamedRerouteUsage()
  n42 = CameraVectorWS()
  n43 = VertexInterpolator(VS=NamedRerouteUsage())
  n44 = Constant3Vector() {constant=<rgba 1 0 0 0>}
  n45 = Normalize(VectorInput=Transform(None=StaticSwitchParameter(True=@n43.PS, False=@n44) {defaultValue=true, parameterName=isRotating}) {transformSourceType=TRANSFORMSOURCE_Local})
  n46 = OneMinus(None=Abs(None=DotProduct(A=@n42, B=@n45)))
  n47 = NamedRerouteUsage()
  n48 = Multiply(
   A=Multiply(
    A=ShadingPathSwitch(Default=@n39, Mobile=@n40)
    B=MaterialFunctionCall(Distance=@n41) {materialFunction=&Near_Camera_Fade}.Result
   )
   B=Saturate(None=Power(Base=@n46, Exp=@n47))
  )
  n49 = CollectionParameter() {collection=/SparksStageShow/Systems/StageManager_v2_Prototype/MaterialFunctions/MPC_SparksStageShow, parameterName=PrimaryColor}
  NamedRerouteDeclaration(None=MaterialFunctionCall(Emissive=Multiply(A=@n48, B=@n49)) {materialFunction=&MF_FX_EmissiveAdjustments}.Result) {name=FinalEmission, nodeColor=<rgba 0 0.625 1 1>}
  n50 = ScalarParameter() {defaultValue=800, desc=Lenght of the light beam, group=Light Beam Settings, parameterName=Length, sliderMax=8000, sortPriority=0}
  NamedRerouteDeclaration(None=@n50) {name=BeamLength, nodeColor=<rgba 0 1 0.0648 1>}
  n51 = ScalarParameter() {
   defaultValue=0.7
   desc=How diffuse the beam gets with distance
   group=Light Beam Settings
   parameterName=Diffusion
   sliderMax=1
   sortPriority=4
  }
  NamedRerouteDeclaration(None=@n51) {name=Diffusion, nodeColor=<rgba 1 0 0.5391 1>}
  n52 = ScalarParameter() {
   defaultValue=3
   desc="Light attenuation coefficient. Lower values makes attenuation more linear.\r\nAttenuation takes a long distance to fall close to zero, so use TipFade property to adjust the beam termination"
   group=Light Beam Settings
   parameterName=Attenuation
   sliderMax=5
   sortPriority=6
  }
  NamedRerouteDeclaration(None=@n52) {name=Attenuation, nodeColor=<rgba 0.8984 0 1 1>}
  NamedRerouteDeclaration(None=ScalarParameter() {defaultValue=25, desc=Light beam base width, group=Light Beam Settings, parameterName=Beam Width, sortPriority=1}) {name=Beam Width, nodeColor=<rgba 0 1 0.0648 1>}
  n53 = ScalarParameter() {defaultValue=0.015, desc=How wide the beam projects, group=Light Beam Settings, parameterName=BeamAngle, sliderMax=0.2, sortPriority=2}
  NamedRerouteDeclaration(None=@n53) {name=Beam Angle, nodeColor=<rgba 0 0.7891 1 1>}
  n54 = ScalarParameter() {
   defaultValue=0.2
   desc="Sets how smoothly the beam disappears at the end tip.\r\nManly used to avoid a harsh line where the beam ends."
   group=Light Beam Settings
   parameterName=TipFade
   sliderMax=1
   sortPriority=7
  }
  NamedRerouteDeclaration(None=@n54) {name=Tip Fade, nodeColor=<rgba 0 1 0.8828 1>}
  n55 = ScalarParameter() {defaultValue=20, desc=Brightness intensity, group=Light Beam Settings, parameterName=Intensity, sliderMax=50, sortPriority=3}
  NamedRerouteDeclaration(None=@n55) {name=Intensity, nodeColor=<rgba 0.9047 0 0.6785 1>}
  NamedRerouteDeclaration(None=MaterialFunctionCall!(Beats Per Measure) {materialFunction=/SparksStageShow/Systems/StageManager_v2_Prototype/MaterialFunctions/MF_SparksStageShow_IntensitySignal}.Signal) {name=SS Intensity Signal, nodeColor=<rgba 0 0.9766 1 1>}
  NamedRerouteUsage()
  NamedRerouteUsage()
 ]
}
