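# Motorized spotlight material for the SparksStageShow prototype, turning masked mesh parts around a pivot to aim along a beat-driven direction.
# A rotation-signal pattern, oriented by a per-primitive resting direction, builds a look-at basis for vertex offset and normals.
# Custom UVs carry rotated normals and aim direction to pixels for normal blending and view-facing emissive boost, tinted by the collection's PrimaryColor and pulsed by an intensity signal.
# Axis-angle rotation and reflection-based UV depth chains appear unused.
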Material /SparksStageShow/Systems/StageManager_v2_Prototype/Materials/M_Spotlight_Rotating_LookAt
graph {
 properties {bAutomaticallySetUsageInEditor=false, maxWorldPositionOffsetDisplacement=50, numCustomizedUVs=4, usedShadingModels=Default Lit}
 nodes [
  n1 = TextureSample() {samplerType=SAMPLERTYPE_Normal, texture=/Game/Athena/Environments/Festivus/MotorizedLights/T_MotorizedSpotLight_01_N}
  n2 = MaterialFunctionCall!(
   Normal=MaterialFunctionCall(R=Multiply(A=@n1.R), G=Multiply(A=@n1.G), B=Constant() {r=1}) {materialFunction=&AppendMany}.RGB
   CameraVector=Transform(None=CameraVectorWS()) {transformSourceType=TRANSFORMSOURCE_World, transformType=TRANSFORM_Tangent}
  ) {materialFunction=/Engine/Functions/Engine_MaterialFunctions02/Utility/CustomReflectionVector}
  n3 = ComponentMask(None=@n2.Reflection Vector) {g=true, r=true}
  n4 = Divide(A=ScalarParameter() {defaultValue=35, parameterName=Depth}, B=Abs(None=ComponentMask(None=@n2.Reflection Vector) {b=true}))
  n5 = Constant() {r=1}
  n6 = Constant() {r=1024}
  Add(
   A=Multiply(A=ScalarParameter() {defaultValue=1, parameterName=UVTile_Depth}, B=TextureCoordinate())
   B=Multiply(A=Multiply(A=@n3, B=@n4), B=Divide(A=@n5, B=@n6))
  )
  n7 = NamedRerouteUsage()
  n8 = Normalize(VectorInput=Multiply(A=Constant3Vector() {constant=<rgba 1 1 0 0>}, B=@n7))
  n9 = Constant3Vector() {constant=<rgba 0 0 1 0>}
  n10 = Constant3Vector() {constant=<rgba 0 -1 0 0>}
  n11 = Sign(None=ComponentMask(None=@n7) {b=true})
  n12 = Multiply(
   A=MaterialFunctionCall(Vector=CrossProduct(A=@n8, B=@n9), Default=Multiply(A=@n11, B=@n10)) {materialFunction=&SafeNormalize}.Result
   B=@n11
  )
  n13 = ArccosineFast!(None=DotProduct(A=@n8, B=@n7))
  n14 = Multiply(A=Multiply(A=Multiply(A=@n13) {constB=57.29, desc=Radians to Degrees}) {constB=0.0028}, B=VertexColor() {desc=Rotation mask}.G)
  n15 = VertexNormalWS()
  n16 = MaterialFunctionCall(Float3=@n7) {materialFunction=&BreakOutFloat3Components}
  n17 = Arctangent2Fast(Y=@n16.G, X=@n16.G)
  n18 = Multiply(A=VertexColor() {desc=Rotation mask}.B, B=Multiply(A=Multiply(A=@n17) {constB=57.29, desc=Radians to Degrees}) {constB=0.0028})
  MaterialFunctionCall(
   Rotation Axis=@n12
   Rotation Angle=@n14
   World Space Vertex Normals=MaterialFunctionCall(
    Rotation Axis=@n9
    Rotation Angle=@n18
    World Space Vertex Normals=Transform(None=@n15) {transformSourceType=TRANSFORMSOURCE_World, transformType=TRANSFORM_Local}
   ) {materialFunction=&FixRotateAboutAxisNormals}.New World Space Vertex Normal
  ) {materialFunction=&FixRotateAboutAxisNormals}
  NamedRerouteDeclaration(None=ComponentMask(None=VectorParameter() {group=Spotlight, parameterName=RotationPivot, sortPriority=1, channelNames {a=-, b=Z, g=Y, r=X}}) {b=true, g=true, r=true}) {name=RotationPivot, nodeColor=<rgba 0.1283 0.1263 0.1302 1>}
  n19 = MaterialFunctionCall(Float2=TextureCoordinate() {coordinateIndex=1}) {materialFunction=&BreakOutFloat2Components}
  n20 = MaterialFunctionCall(
   X=@n19.R
   Y=@n19.R
   Z=MaterialFunctionCall(Float2=TextureCoordinate() {coordinateIndex=2}) {materialFunction=&BreakOutFloat2Components}.R
  ) {materialFunction=&MakeFloat3}
  NamedRerouteDeclaration(None=@n20.Result) {name=RotatedNormalPS, nodeColor=<rgba 0 0.7656 1 1>}
  n21 = Normalize(VectorInput=Transform(None=NamedRerouteUsage()) {transformSourceType=TRANSFORMSOURCE_Local, transformType=TRANSFORM_Tangent})
  NamedRerouteDeclaration(None=ComponentMask(None=@n21) {g=true, r=true}) {name=CustomUV1, desc="Normals (XY)", nodeColor=<rgba 0.011 0.417 1 1>}
  n22 = Transform(None=NamedRerouteUsage()) {transformSourceType=TRANSFORMSOURCE_Local}
  NamedRerouteDeclaration(None=MaterialFunctionCall(X=ComponentMask(None=@n21) {b=true}, Y=ComponentMask(None=@n22) {r=true}) {materialFunction=&MakeFloat2}.Result) {name=CustomUV2, desc="Normals (Z)\r\nFinal taget direction (X)", nodeColor=<rgba 0 1 0.1562 1>}
  NamedRerouteDeclaration(None=ComponentMask(None=@n22) {b=true, g=true}) {name=CustomUV3, desc="Final target direction (YZ)", nodeColor=<rgba 1 0 0.1641 1>}
  n23 = CollectionParameter() {collection=/SparksStageShow/Systems/StageManager_v2_Prototype/MaterialFunctions/MPC_SparksStageShow, parameterName=PrimaryColor}
  n24 = ComponentMask(None=TextureCoordinate() {coordinateIndex=2}) {g=true}
  n25 = MaterialFunctionCall(Float2=TextureCoordinate() {coordinateIndex=3}) {materialFunction=&BreakOutFloat2Components}
  n26 = Constant3Vector() {constant=<rgba 1 0 0 0>}
  n27 = DotProduct(
   A=CameraVectorWS()
   B=StaticSwitchParameter(
    True=MaterialFunctionCall(X=@n24, Y=@n25.R, Z=@n25.R) {materialFunction=&MakeFloat3}.Result
    False=Transform(None=@n26) {transformSourceType=TRANSFORMSOURCE_Local}
   ) {defaultValue=true, group=Spotlight, parameterName=IsRotating}
  )
  n28 = LinearInterpolate(A=NamedRerouteUsage(), B=NamedRerouteUsage(), Alpha=Power(Base=Saturate(None=@n27)) {constExponent=6})
  n29 = NamedRerouteUsage()
  n30 = Multiply(
   A=Multiply(A=Saturate(None=@n23), B=Multiply(A=@n28, B=@n29))
   B=VertexInterpolator(VS=MaterialFunctionCall!(Beats Per Measure) {materialFunction=/SparksStageShow/Systems/StageManager_v2_Prototype/MaterialFunctions/MF_SparksStageShow_IntensitySignal}.Signal).PS
  )
  n31 = TextureSampleParameter2D() {parameterName=Emissive, texture=/Game/Athena/Environments/Festivus/MotorizedLights/T_MotorizedSpotLight_01_M}
  NamedRerouteDeclaration(None=MaterialFunctionCall(Emissive=Multiply(A=@n30, B=@n31.RGBA)) {materialFunction=&MF_FX_EmissiveAdjustments}.Result) {name=Emissive, nodeColor=<rgba 1 0 0.7984 1>}
  n32 = VectorParameter() {
   bUseCustomPrimitiveData=true
   group=Spotlight
   parameterName=LocalRestingDirection
   primitiveDataIndex=18
   sortPriority=0
   channelNames {a=-, b=Z, g=Y, r=X}
   defaultValue=<rgba 1 0 0 0>
  }
  NamedRerouteDeclaration(None=@n32) {name=LocalRestingDir, nodeColor=<rgba 0.3828 0 1 1>}
  NamedRerouteDeclaration(None=ScalarParameter() {defaultValue=1, group=Spotlight, parameterName=ViewDirBoostMin, sortPriority=3}) {name=ViewDirBoostMin, nodeColor=<rgba 0 0.8594 1 1>}
  NamedRerouteDeclaration(None=ScalarParameter() {defaultValue=5, group=Spotlight, parameterName=ViewDirBoostMax, sortPriority=4}) {name=ViewDirBoostMax, nodeColor=<rgba 1 0.8203 0 1>}
  NamedRerouteDeclaration(None=ScalarParameter() {defaultValue=20, group=Spotlight, parameterName=Intensity, sortPriority=2}) {name=Intensity, nodeColor=<rgba 0.5 0 1 1>}
  n33 = MaterialFunctionCall!(
   Normalized Rotation Pattern=Normalize(VectorInput=MaterialFunctionCall!(Beats Per Measure) {materialFunction=/SparksStageShow/Systems/StageManager_v2_Prototype/MaterialFunctions/MF_SparksStageShow_RotationSignal}.Result)
   Normalized Direction=Normalize(VectorInput=NamedRerouteUsage())
  ) {materialFunction=/SparksStageShow/Systems/StageManager_v2_Prototype/MaterialFunctions/MF_SparksStageShow_TransformRotPattern}
  NamedRerouteDeclaration(None=@n33.Transformed Pattern) {name=TargetDirection, nodeColor=<rgba 1 0.8324 0 1>}
  NamedRerouteDeclaration(None=@n33.BasisY) {name=BasisY, nodeColor=<rgba 0 1 0.0891 1>}
  NamedRerouteDeclaration(None=@n33.BasisX) {name=BasisX, nodeColor=<rgba 1 0 0.0028 1>}
  n34 = Subtract(A=LocalPosition().XYZ, B=NamedRerouteUsage())
  n35 = VertexColor()
  n36 = Multiply(A=NamedRerouteUsage(), B=Constant3Vector() {constant=<rgba 1 1 0 0>})
  n37 = MaterialFunctionCall(
   Vector=Multiply(A=NamedRerouteUsage(), B=@n35)
   Default=MaterialFunctionCall(Vector=Multiply(A=@n35, B=@n36), Default=Constant3Vector() {constant=<rgba 1 0 0 0>}) {materialFunction=&SafeNormalize}.Result
  ) {materialFunction=&SafeNormalize}
  n38 = MaterialFunctionCall(Vector=CrossProduct(A=Constant3Vector() {constant=<rgba 0 0 1 0>}, B=@n37.Result), Default=NamedRerouteUsage()) {materialFunction=&SafeNormalize}
  n39 = CrossProduct(A=@n37.Result, B=@n38.Result)
  n40 = MaterialFunctionCall(VectorToTransform=@n34, BasisX=@n37.Result, BasisY=@n38.Result, BasisZ=@n39) {materialFunction=&Transform3x3Matrix}
  NamedRerouteDeclaration(None=Transform(None=Subtract(A=@n40.Transformed Vector, B=@n34)) {transformSourceType=TRANSFORMSOURCE_Local}) {name=VertexOffset, nodeColor=<rgba 0.4273 0.0985 0.5729 1>}
  n41 = VertexNormalWS()
  NamedRerouteDeclaration(None=MaterialFunctionCall(
    VectorToTransform=Transform(None=@n41) {transformSourceType=TRANSFORMSOURCE_World, transformType=TRANSFORM_Local}
    BasisX=@n37.Result
    BasisY=@n38.Result
    BasisZ=@n39
   ) {materialFunction=&Transform3x3Matrix}.Transformed Vector) {name=RotatedNormalVS, nodeColor=<rgba 0.0005 0.7275 1 1>}
  n42 = TextureSampleParameter2D() {parameterName=Normal, samplerType=SAMPLERTYPE_Normal, texture=&T_MotorizedSpotLight_01_N}
  NamedRerouteDeclaration(None=StaticSwitchParameter(
    True=MaterialFunctionCall(BaseNormal=@n20.Result, AdditionalNormal=@n42.RGB) {materialFunction=&BlendAngleCorrectedNormals}.Result
    False=@n42.RGB
   ) {defaultValue=true, group=Spotlight, parameterName=IsRotating}) {name=BlendedNormals, nodeColor=<rgba 0 0.4384 1 1>}
  NamedRerouteUsage()
  NamedRerouteUsage()
  NamedRerouteUsage()
  NamedRerouteUsage()
  NamedRerouteUsage()
  StaticSwitchParameter(True=NamedRerouteUsage(), False=Constant3Vector()) {defaultValue=true, group=Spotlight, parameterName=IsRotating}
  TextureSampleParameter2D() {parameterName=Base Color, texture=/Game/Athena/Environments/Festivus/MotorizedLights/T_MotorizedSpotLight_01_D}
  n43 = NamedRerouteUsage()
  n44 = RotateAboutAxis(NormalizedRotationAxis=@n9, RotationAngle=@n18, PivotPoint=NamedRerouteUsage(), Position=LocalPosition().XYZ)
  n45 = Add(A=@n44, B=LocalPosition().XYZ)
  Transform(None=Add(A=@n44, B=RotateAboutAxis(NormalizedRotationAxis=@n12, RotationAngle=@n14, PivotPoint=@n43, Position=@n45))) {transformSourceType=TRANSFORMSOURCE_Local}
 ]
}
